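# Looping Sound Cue for the ThornRipper chain whip hookshot, using attenuation and melee weapon concurrency.
# A chain loop layered with a quieter ghost-trail loop gets oscillator volume wobble, then islocal and isenemy branches halve its volume for remote non-enemies.
# In parallel, one of two voice loops is picked at random and faded in over 1.5 seconds by an envelope, bypassing those branches.
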SoundCue /ThornRipper/Assets/Audio/ThornRipper_ChainWhip_Hookshot_Loop_Cue
root=Mixer_2
nodes [
 {id=Branch_2, children [Modulator_2,Branch_3,Modulator_2], properties {boolParameterName=islocal}}
 {id=Branch_3, children [Modulator_4,Modulator_3,Modulator_4], properties {boolParameterName=isenemy}}
 {
  id=Enveloper_0
  children [Modulator_0]
  properties {
   pitchCurve {editorCurveData {keys [{time=0}]}}
   volumeCurve {editorCurveData {keys [
      {interpMode=RCIM_Cubic, tangentMode=RCTM_User, time=0, value=0}
      {arriveTangent=0.3061, interpMode=RCIM_Cubic, leaveTangent=0.3061, tangentMode=RCTM_User, time=1.5, value=1}
     ]}}
  }
 }
 {id=Mixer_1, children [WavePlayer_0,WavePlayer_2], properties {inputVolume [2,0.3]}}
 {id=Mixer_2, children [Branch_2,Enveloper_0], properties {inputVolume [1,1]}}
 {id=Modulator_0, children [Random_0]}
 {id=Modulator_1, children [QualityLevel_1], properties {pitchMax=1, pitchMin=1}}
 {id=Modulator_2, children [Modulator_1], properties {pitchMax=1, pitchMin=1, volumeMax=1, volumeMin=1}}
 {id=Modulator_3, children [Modulator_1], properties {pitchMax=1, pitchMin=1, volumeMax=0.5, volumeMin=0.5}}
 {id=Modulator_4, children [Modulator_1], properties {pitchMax=1, pitchMin=1, volumeMax=1, volumeMin=1}}
 {
  id=Oscillator_0
  children [Mixer_1]
  properties {amplitudeMax=0.3, amplitudeMin=0.3, bModulateVolume=true, centerMax=1, centerMin=1, frequencyMax=5, frequencyMin=5}
 }
 {id=QualityLevel_1, children [Oscillator_0,Oscillator_0,Oscillator_0,Oscillator_0]}
 {id=Random_0, children [WavePlayer_1,WavePlayer_3], properties {weights [1,1]}}
 {
  id=WavePlayer_0
  properties {bLooping=true, soundWaveAssetPtr=/SunRoseChainWhipGameplay/Items/ChainWhip/Sound/ChainWhip_HookShot_Loop_Chains}
 }
 {
  id=WavePlayer_1
  properties {bLooping=true, soundWaveAssetPtr=/SunRoseChainWhipGameplay/Items/ChainWhip/Sound/ChainWhip_HookShot_Voices_Loop_01}
 }
 {id=WavePlayer_2, properties {bLooping=true, soundWaveAssetPtr=/Game/Sounds/Fort_Weapons/pistol/Ghost/SFX_GhostPistol_GhostTrail_Loop_03}}
 {
  id=WavePlayer_3
  properties {bLooping=true, soundWaveAssetPtr=/SunRoseChainWhipGameplay/Items/ChainWhip/Sound/ChainWhip_HookShot_Voices_Loop_02}
 }
]
properties {
 attenuationSettings=&/ThornRipper_ChainWhip_Prj_Twirl_Loop_Att
 maxDistance=6500
 soundClassObject=&Fort_WeaponMelee_PSM
 volumeMultiplier=1
 attenuationOverrides {distanceAlgorithm=NaturalSound, falloffDistance=3000, attenuationShapeExtents {x=1000}}
 concurrencySet [&Player_Melee_Weapon_Conc]
}
